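# Additive unlit Niagara ribbon material for wavy edge glow lines on the Cactus Rocker male pickaxe FX.
# Two T_LaserTrail samples get tiled, V-clamped UVs bent by sine waves along U, with a particle dynamic parameter scaling only the second wave's amplitude.
# Their summed green channels are smoothstepped into lines and multiplied by particle alpha, with depth fade on PC but not Switch or mobile.
# Particle color feeds emissive through MF_FX_EmissiveAdjustments.
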Material /BRCosmetics/Weapons/FORT_Melee/Pickaxe_Cactus_Rocker_Male/FX/M_Ribbon_EdgeGlows
graph {
 properties {bUsedWithNiagaraRibbons=true, blendMode=BLEND_Additive, pixelDepthOffsetMode=PDOM_Legacy, shadingModel=MSM_Unlit, usedShadingModels=Unlit}
 nodes [
  n1 = ParticleColor()
  MaterialFunctionCall(Emissive=@n1.RGB) {materialFunction=&MF_FX_EmissiveAdjustments}
  n2 = ScalarParameter() {defaultValue=1, parameterName=SineATileU}
  n3 = ScalarParameter() {defaultValue=2, parameterName=SineATileV}
  n4 = MaterialFunctionCall(Float2=Multiply(A=TextureCoordinate(), B=AppendVector(A=@n2, B=@n3))) {materialFunction=&BreakOutFloat2Components}
  n5 = AppendVector(A=@n4.R, B=Saturate(None=@n4.G))
  n6 = Constant()
  n7 = ComponentMask(None=@n5) {r=true}
  n8 = ScalarParameter() {defaultValue=2, parameterName=SineAFreq}
  n9 = ScalarParameter() {defaultValue=0.1, parameterName=SineAAmp}
  n10 = Multiply(A=Sine(None=Multiply(A=@n7, B=@n8)), B=@n9)
  n11 = Add(
   A=Add(A=@n5, B=AppendVector(A=@n6, B=@n10))
   B=AppendVector(A=ScalarParameter() {parameterName=SineAUOffset}, B=ScalarParameter() {parameterName=SineAVOffset})
  )
  n12 = TextureCoordinate()
  n13 = AppendVector(A=ScalarParameter() {defaultValue=1, parameterName=SineBTileU}, B=ScalarParameter() {defaultValue=2, parameterName=SineBTileV})
  n14 = ScalarParameter() {parameterName=SineBUOffset}
  n15 = ScalarParameter() {defaultValue=-1, parameterName=SineBVOffset}
  n16 = MaterialFunctionCall(Float2=Add(A=Multiply(A=@n12, B=@n13), B=AppendVector(A=@n14, B=@n15))) {materialFunction=&BreakOutFloat2Components}
  n17 = AppendVector(A=@n16.R, B=Saturate(None=@n16.G))
  n18 = Sine(None=Multiply(A=ComponentMask(None=@n17) {r=true}, B=ScalarParameter() {defaultValue=2.5, parameterName=SineBFreq}))
  n19 = Multiply(
   A=ScalarParameter() {defaultValue=0.15, parameterName=SineBAmp}
   B=DynamicParameter() {paramNames [SineWaveDistAmt, Param2, Param3, Param4]}.SineWaveDistAmt
  )
  n20 = Add(A=@n17, B=AppendVector(A=Constant(), B=Multiply(A=@n18, B=@n19)))
  n21 = SmoothStep(
   Min=ScalarParameter() {defaultValue=0.5, parameterName=LineMin}
   Max=ScalarParameter() {defaultValue=0.6, parameterName=LineMax}
   Value=Add(
    A=TextureSample(UVs=@n11) {texture=/Game/Effects/Fort_Effects/Textures/Tile/T_LaserTrail}.G
    B=TextureSample(UVs=@n20) {texture=&T_LaserTrail}.G
   )
  ) {constMax=0.5, constMin=0.3}
  n22 = DepthFade(Opacity=@n21, FadeDistance=ScalarParameter() {defaultValue=100, group="6.1) Opacity", parameterName=Dpeth Fade Distance, sortPriority=10})
  n23 = ShadingPathSwitch(Default=@n22, Deferred=@n22, Forward=@n21, Mobile=@n21) {desc="Defered is pc, Forward is switch, mobile is mobile"}
  Multiply(A=@n1.A, B=ShadingPathSwitch(Default=@n23, Deferred=@n23, Mobile=@n21) {desc=Migrated from FeatureLevelSwitch})
  Multiply(A=@n9, B=@n19)
 ]
}
