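# Idle Niagara effect for the Cactus Rocker Male pickaxe: a single audio-reactive, color-shifting light, offset from the origin, with no visible sprite or mesh.
# Event audio amplitude from the Fortnite NPC (Niagara parameter collection) is scaled, clamped and smoothed into an emitter-level gate.
# That gate blends between slow and fast age-driven curve sampling for hue and brightness flicker, and also widens the light radius.
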NiagaraSystem /BRCosmetics/Weapons/FORT_Melee/Pickaxe_Cactus_Rocker_Male/FX/NS_Pickaxe_CactusRockerMale_Idle
dependencies {usedDataInterfaces [&NiagaraDataInterfaceCurve], usedDynamicInputs [], usedModules [&SystemState]}
emitters {Light {
  emitterSpawnScript [/SetVariables() {isSetParameters=true}]
  emitterUpdateScript [
   &EmitterState()
   &SpawnBurst_Instantaneous(Spawn Count=1, Use Spawn Probability=false)
   /SetVariables(Emitter.IncomingAudioGate:NiagaraFloat=/Niagara/DynamicInputs/Lerp/SmoothLerpOverTime_Float(
     Convergence Rate:NiagaraFloat=2
     Smooth Value:NiagaraFloat=$Emitter.IncomingAudioGate
     Target Value:NiagaraFloat=&ClampFloat(Float=&Multiply_Float(A=$NPC.FortniteNPC.EventAudio_Amplitude, B=15))
    )) {isSetParameters=true}
  ]
  particleSpawnScript [&InitializeParticle(Color Mode=::NewEnumerator0, Lifetime=10, Position Offset=<5 0 30>, UsePositionOffset=true, Write Material Random=true) {module=InitializeParticle001}]
  particleUpdateScript [
   &ParticleState()
   &Color(Color=&MakeLinearColorFromVectorAndFloat(
     "Vector (RGB)"=&Multiply_Vector(
      A=&MakeVectorFromLinearColorRGB(Color=&HueShiftLinearColor(
        Hue=&FloatFromCurve(
         FloatCurve=DataInterface() {propertyValues {Curve {keys [<at 0 1>,<at 0.1069 0.0895>,<at 0.3206 0.7423>,<at 0.4907 0.1444>,<at 0.6992 0.8948>,<at 0.9061 0.1261>,<at 1 1>]}}}
         CurveIndex=&Lerp_Float(
          A=&Frac_Float(Float=&Multiply_Float(A=$Emitter.Age, B=0.03))
          B=&Frac_Float(Float=&Multiply_Float(A=$Emitter.Age, B=0.15))
          Alpha=$Emitter.IncomingAudioGate
         )
        )
        Saturation=0.4
        Value=1
        Linear Color=&MakeLinearColor(R:NiagaraFloat=1, G:NiagaraFloat=0, B:NiagaraFloat=0, A:NiagaraFloat=0)
       ))
      B=&VectorFromFloat(Value=&FloatFromCurve(
        FloatCurve=DataInterface() {propertyValues {Curve {keys [<at 0 1>,<at 0.0584 1.1216>,<at 0.0797 0.6691>,<at 0.1464 0.526>,<at 0.222 1.0327>,<at 0.2705 0.6844>,<at 0.3166 0.8798>,<at 0.3641 0.5477>,<at 0.4407 0.6405>,<at 0.4815 1.1358>,<at 0.5388 1.0006>,<at 0.58 0.5314>,<at 0.6479 0.8407>,<at 0.6937 0.7459>,<at 0.737 1.0896>,<at 0.8125 0.5703>,<at 0.8909 0.8732>,<at 0.9449 0.6186>,<at 1 1>]}}}
        CurveIndex=&Lerp_Float(
         A=&Frac_Float(Float=&Multiply_Float(A=$Emitter.Age, B=0.01))
         B=&Frac_Float(Float=&Multiply_Float(A=$Emitter.Age, B=0.3))
         Alpha=$Emitter.IncomingAudioGate
        )
       ))
     )
     "Float (Alpha)"=1
    ))
   &FN_Light_v2(
    Light Exponent=2
    Light Radius:NiagaraFloat=&Lerp_Float(A=50, B=65, Alpha=$Emitter.IncomingAudioGate)
    LightEnabled=true
    Scale Color=10
   )
  ]
  rendererClasses [&NiagaraLightRendererProperties]
  renderers [{rendererClass=&NiagaraLightRendererProperties, rendererIndex=0}]
 }}
summary {
 systemName=NS_Pickaxe_CactusRockerMale_Idle
 emitters [{emitterName=Light, rendererClasses [&NiagaraLightRendererProperties]}]
 userVariables [
  User.bIsFrontEnd {type=NiagaraBool}
  User.bIsFrontEndPreview {type=NiagaraBool}
 ]
}
