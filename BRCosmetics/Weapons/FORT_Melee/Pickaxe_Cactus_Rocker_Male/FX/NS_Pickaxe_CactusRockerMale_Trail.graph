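# Niagara swing trail for the Cactus Rocker Male pickaxe, drawing anim-trail ribbons from bones or sockets on a user-supplied skeletal mesh.
# Both emitters fade in over emitter loop age, then shift through rainbow colors and fade out over short particle lifetimes.
# FN_AnimTrail uses two ribbon renderers and a half-width secondary value.
# FN_SecondaryRibbons adds a thinner, brighter ribbon with age-driven U offset and a curve-driven material parameter.
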NiagaraSystem /BRCosmetics/Weapons/FORT_Melee/Pickaxe_Cactus_Rocker_Male/FX/NS_Pickaxe_CactusRockerMale_Trail
dependencies {
 usedDataInterfaces [&NiagaraDataInterfaceColorCurve, &NiagaraDataInterfaceCurve]
 usedDynamicInputs []
 usedModules [&SystemState]
}
emitters {
 FN_AnimTrail {
  emitterUpdateScript [
   &EmitterState()
   &FN_SpawnAnimTrail_v2()
  ]
  particleSpawnScript [
   /Game/Effects/Niagara/Modules/AnimTrail/FN_InitializeAnimTrail/FN_InitializeAnimTrailBoneIndecies(Second Bone or Socket Index=1)
   /Game/Effects/Niagara/Modules/AnimTrail/FN_InitializeAnimTrail/FN_InitializeAnimTrail(
    AnimTrailWidth=1.5
    Color=&ColorFromCurve(
     ColorCurve=DataInterface() {propertyValues {
       AlphaCurve {keys [<at 0.0249 0 RCIM_Cubic>,<at 0.0932 1>,<at 1 1 RCIM_Cubic>]}
       BlueCurve {keys [<at 0 1 RCIM_Cubic>,<at 1 1 RCIM_Cubic>]}
       GreenCurve {keys [<at 0 1 RCIM_Cubic>,<at 1 1 RCIM_Cubic>]}
       RedCurve {keys [<at 0 1 RCIM_Cubic>,<at 1 1 RCIM_Cubic>]}
      }}
     CurveIndex=$Emitter.NormalizedLoopAge
    )
    Lifetime=0.2
   )
  ]
  particleUpdateScript [
   &FN_UpdateAnimTrail_v2()
   &ScaleColor(
    Linear Color Curve=DataInterface() {propertyValues {
      AlphaCurve {keys [<at 0 1>,<at 0.2 0.7>,<at 1 0>]}
      BlueCurve {keys [<at 0.0742 0.1198>,<at 0.2833 0.125>,<at 0.4958 0.956>,<at 0.7033 1>,<at 0.909 0.4689>]}
      GreenCurve {keys [<at 0.0742 0.7359>,<at 0.2833 1>,<at 0.4958 1>,<at 0.7033 0.125>,<at 0.909 0.1615>]}
      RedCurve {keys [<at 0.0742 1>,<at 0.2833 0.1687>,<at 0.4958 0.1198>,<at 0.7033 0.4604>,<at 0.909 1>]}
     }}
    Scale Mode=::NewEnumerator2
   )
   /SetVariables(Particles.SecondWidth:NiagaraFloat=&Multiply_Float(A=$Particles.RibbonWidth, B=0.5)) {isSetParameters=true}
  ]
  rendererClasses [&NiagaraRibbonRendererProperties]
  renderers [
   {rendererClass=&NiagaraRibbonRendererProperties}
   {rendererClass=&NiagaraRibbonRendererProperties, rendererIndex=1}
  ]
 }
 FN_SecondaryRibbons {
  emitterUpdateScript [
   &EmitterState()
   &FN_SpawnAnimTrail_v2()
  ]
  particleSpawnScript [
   &FN_InitializeAnimTrailBoneIndecies(Second Bone or Socket Index=1)
   &FN_InitializeAnimTrail(
    AnimTrailWidth=0.5
    Color=&ColorFromCurve(
     ColorCurve=DataInterface() {propertyValues {
       AlphaCurve {keys [<at 0.0704 0 RCIM_Cubic>,<at 0.1388 1>,<at 1 1 RCIM_Cubic>]}
       BlueCurve {keys [<at 0 5 RCIM_Cubic>]}
       GreenCurve {keys [<at 0 5 RCIM_Cubic>]}
       RedCurve {keys [<at 0 5 RCIM_Cubic>]}
      }}
     CurveIndex=$Emitter.NormalizedLoopAge
    )
    Lifetime=0.15
   )
   /SetVariables(Particles.RibbonU0Override=&Multiply_Float(A=$Emitter.Age, B=20)) {isSetParameters=true}
  ]
  particleUpdateScript [
   &FN_UpdateAnimTrail_v2()
   &ScaleColor(
    Linear Color Curve=DataInterface() {propertyValues {
      AlphaCurve {keys [<at 0 1>,<at 0.2 0.7>,<at 1 0>]}
      BlueCurve {keys [<at 0.0742 0.1198>,<at 0.2833 0.125>,<at 0.4958 0.956>,<at 0.7033 1>,<at 0.909 0.4689>]}
      GreenCurve {keys [<at 0.0742 0.7359>,<at 0.2833 1>,<at 0.4958 1>,<at 0.7033 0.125>,<at 0.909 0.1615>]}
      RedCurve {keys [<at 0.0742 1>,<at 0.2833 0.1687>,<at 0.4958 0.1198>,<at 0.7033 0.4604>,<at 0.909 1>]}
     }}
    Scale Mode=::NewEnumerator2
   )
   &DynamicMaterialParameters(
    Index 0 Param 1=&FloatFromCurve(FloatCurve=DataInterface() {propertyValues {Curve {keys [<at 0 0 RCIM_Cubic>, {arriveTangent=6.2455, interpMode=RCIM_Cubic, leaveTangent=6.2455, tangentMode=RCTM_User, time=1, value=2}]}}})
    Module.Param0WriteEnabled=true
    Module.Param1WriteEnabled=true
    Module.Param2WriteEnabled=true
    Module.Param3WriteEnabled=true
   )
  ]
  rendererClasses [&NiagaraRibbonRendererProperties]
  renderers [{rendererClass=&NiagaraRibbonRendererProperties}]
 }
}
summary {
 systemName=NS_Pickaxe_CactusRockerMale_Trail
 emitters [
  {emitterName=FN_AnimTrail, rendererClasses [&NiagaraRibbonRendererProperties]}
  {emitterName=FN_SecondaryRibbons, rendererClasses [&NiagaraRibbonRendererProperties]}
 ]
 userVariables [
  User.RibbonIDOffset {type=NiagaraInt32}
  User.SkeletalMesh {
   type=NiagaraDataInterfaceSkeletalMesh
   defaultValue {
    struct=NiagaraExt_VariableValue_DataInterface
    value {dataInterface=&self:NiagaraDataInterfaceSkeletalMesh_0}
   }
  }
 ]
}
