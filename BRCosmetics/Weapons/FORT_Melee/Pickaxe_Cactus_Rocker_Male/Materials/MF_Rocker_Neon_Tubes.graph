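# Material function adding audio-reactive neon effects to the male Cactus Rocker pickaxe, toggled by a static switch.
# Event audio amplitude and spectrum, or a fake-data parameter, are capped and shift a scrolling hue gradient built from pre-skinned local position.
# Alpha-input channel masks separate tubes, body and eyes for layered attribute blending.
# Tubes get hue-cycled color, Fresnel-shaped emissive and cosine-wave vertex offset, body gets Fresnel glow, eyes get pulsing emissive.
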MaterialFunction /BRCosmetics/Weapons/FORT_Melee/Pickaxe_Cactus_Rocker_Male/Materials/MF_Rocker_Neon_Tubes
graph {nodes [
  n1 = TextureCoordinate()
  Frac(None=Add(A=Time(), B=ComponentMask(None=@n1) {g=true}))
  n2 = NamedRerouteDeclaration(None=FunctionInput() {inputType=FunctionInput_MaterialAttributes}) {name=MA_In, nodeColor=<rgba 0 1 0.8594 1>}
  n3 = PreSkinnedPosition()
  n4 = Divide(A=ShaderStageSwitch(PixelShader=VertexInterpolator(VS=@n3).PS, VertexShader=@n3), B=PreSkinnedLocalBounds().Extents) {constB=200}
  n5 = Rotator(
   Coordinate=Multiply(
    A=StaticComponentMaskParameter(None=@n4) {defaultB=true, defaultG=true, group=Neon, parameterName=Neon Scroll UVs}
    B=ScalarParameter() {defaultValue=0.5, group=Neon, parameterName=Neon UVs Lenght Scale, sliderMax=2}
   )
   Time=ScalarParameter() {defaultValue=-0.25, group=Neon, parameterName=Rotate Neon UVs, sliderMax=1, sliderMin=-1}
  ) {speed=3.1415}
  n6 = Multiply(A=Time(), B=ScalarParameter() {defaultValue=0.5, group=Neon, parameterName=Neon_AudioPanSpeed}) {constB=0.5}
  n7 = ScalarParameter() {defaultValue=0.7, group=Neon, parameterName=Neon_AudioColorPulseStrength}
  n8 = CollectionParameter() {collection=&FortniteMaterialParameters, parameterName=EventAudio_Amplitude}
  n9 = CollectionParameter() {collection=&FortniteMaterialParameters, parameterName=EventAudio_Spectrum}
  n10 = MaterialFunctionCall(Float4=@n9) {materialFunction=&BreakOutFloat4Components}
  n11 = Add(A=Add(A=@n10.R, B=@n10.R), B=Add(A=@n10.B, B=@n10.B))
  n12 = StaticSwitchParameter(True=ScalarParameter() {group=Neon, parameterName=FakeAudioData, sliderMax=30}, False=Add(A=@n8, B=Divide(A=@n11) {constB=4})) {group=SkinFX, parameterName=UseFakeAudioData}
  n13 = ScalarParameter() {defaultValue=1, group=Neon, parameterName=Neon_EventAudioMultiplier}
  n14 = NamedRerouteDeclaration(None=Min(A=Multiply(A=@n12, B=@n13)) {constB=30, desc=prevent crazy values}) {name=AudioSpectrum, nodeColor=<rgba 1 0.8203 0 1>}
  n15 = If(A=@n14, "A > B"=Add(A=@n6, B=Multiply(A=@n7, B=@n14) {constB=0.7}) {constB=0.2}, "A < B"=@n6) {constB=0.1}
  n16 = ScalarParameter() {defaultValue=0.5, group=Neon, parameterName=Neon UVs Scroll Speed, sliderMax=1}
  n17 = MaterialFunctionCall(
   Hue Shift Percentage=ComponentMask(None=Panner(Coordinate=@n5, Time=@n15, Speed=@n16) {bFractionalPart=true, speedX=0.5}) {r=true}
   Texture=Constant3Vector() {constant=<rgba 1 0 0 0>}
  ) {materialFunction=&HueShift}
  n18 = Multiply(
   A=MaterialFunctionCall!(Base=@n17.Result, Add=NamedRerouteUsage()) {desc=multiplyAdd by audio, materialFunction=/Engine/Functions/Engine_MaterialFunctions02/Math/MultiplyAdd}.Result
   B=ScalarParameter() {defaultValue=1, group=Neon, parameterName=Neon_EmissiveStrength}
  ) {desc=global Mult}
  n19 = OneMinus(None=Fresnel() {exponent=1})
  n20 = ScalarParameter() {defaultValue=1, group=Neon, parameterName=Fresnel_Exp_Tubes, sliderMax=8}
  n21 = Panner(Coordinate=@n5) {bFractionalPart=true, speedX=0.5}
  n22 = Cosine(None=Multiply(A=ComponentMask(None=@n21) {r=true}) {constB=4})
  n23 = Multiply(
   A=Multiply(A=Constant3Vector() {constant=<rgba 0 0 1 0>}, B=ScalarParameter() {defaultValue=-0.5, group=Neon, parameterName=Normal_Push_Intensity})
   B=@n22
  )
  n24 = NamedRerouteUsage()
  n25 = SetMaterialAttributes!(
   MaterialAttributes=@n2
   BaseColor=@n17.Result
   Metallic=ScalarParameter() {defaultValue=1, group=Neon, parameterName=Metallic, sliderMax=1}
   Specular=ScalarParameter() {defaultValue=1, group=Neon, parameterName=Specular, sliderMax=1}
   Roughness=ScalarParameter() {defaultValue=0.2, group=Neon, parameterName=Roughness, sliderMax=1}
   EmissiveColor=Multiply(A=NamedRerouteDeclaration(None=@n18) {name=EmissiveColor_Reactive, nodeColor=<rgba 1 0 0.0937 1>}, B=Power(Base=@n19, Exp=@n20))
   WorldPositionOffset=Multiply(A=Transform(None=@n23), B=Saturate(None=@n24))
  ) {attributeSetTypes [
    #1
    #2
    #3
    #4
    #5
    #6
   ]}
  n26 = BlendMaterialAttributes(
   A=BlendMaterialAttributes(A=@n2, B=@n25, Alpha=NamedRerouteUsage())
   B=SetMaterialAttributes!(MaterialAttributes=NamedRerouteUsage(), EmissiveColor=NamedRerouteUsage()) {attributeSetTypes [#5]}
   Alpha=NamedRerouteUsage()
  )
  n27 = SetMaterialAttributes(MaterialAttributes=NamedRerouteUsage(), EmissiveColor=NamedRerouteUsage()) {attributeSetTypes [#5]}
  n28 = NamedRerouteUsage()
  FunctionOutput(None=StaticSwitchParameter(True=BlendMaterialAttributes(A=@n26, B=@n27, Alpha=@n28), False=NamedRerouteUsage()) {group=SkinFX, parameterName=UseRockerNeonTubes})
  n29 = FunctionInput() {bUsePreviewValueAsDefault=true, inputName=Alpha, inputType=FunctionInput_Vector4, sortPriority=1}
  NamedRerouteDeclaration(None=ChannelMaskParameter(None=@n29) {group=SkinFX, parameterName=Rocker_NeonTubes_AlphaChannel}.R) {name=Tubes_Alpha, nodeColor=<rgba 1 0.5 0 1>}
  n30 = ChannelMaskParameter(None=@n29) {group=SkinFX, maskChannel=Green, parameterName=Rocker_Body_AlphaChannel, defaultValue=<rgba 0 1 0 0>}
  NamedRerouteDeclaration(None=@n30.R) {name=Body_Alpha, nodeColor=<rgba 1 0.5 0 1>}
  n31 = ChannelMaskParameter(None=@n29) {group=SkinFX, maskChannel=Blue, parameterName=Rocker_Eyes_AlphaChannel, defaultValue=<rgba 0 0 1 0>}
  NamedRerouteDeclaration(None=@n31.R) {name=Eyes_Alpha, nodeColor=<rgba 1 0.5 0 1>}
  n32 = ScalarParameter() {defaultValue=2, group=Neon, parameterName=BodyFresnelExp, sliderMax=8}
  n33 = NamedRerouteUsage()
  n34 = Multiply(
   A=Saturate(None=Fresnel(ExponentIn=@n32) {baseReflectFraction=0, exponent=2})
   B=LinearInterpolate(
    A=ScalarParameter() {defaultValue=0.1, group=Neon, parameterName=BodyFresnel_Intensity, sliderMax=1}
    B=ScalarParameter() {defaultValue=0.5, group=Neon, parameterName=BodyFresnel_Intensity_Active, sliderMax=1}
    Alpha=Saturate(None=@n33)
   )
  )
  n35 = NamedRerouteUsage()
  NamedRerouteDeclaration(None=Multiply(A=NamedRerouteUsage(), B=Multiply(A=@n34, B=@n35))) {name=Body_Emissive, nodeColor=<rgba 0.0547 0 1 1>}
  n36 = GetMaterialAttributes(None=NamedRerouteUsage()) {attributeGetTypes [#5]}
  n37 = Frac(None=Time())
  n38 = Multiply(A=Cosine(None=Multiply(A=@n37) {constB=4})) {constB=0.5}
  n39 = Multiply(A=Saturate(None=Add(A=@n38) {constB=0.5})) {constB=5}
  n40 = NamedRerouteUsage()
  NamedRerouteDeclaration(None=LinearInterpolate(A=@n36.EmissiveColor, B=Multiply(A=@n36.EmissiveColor, B=@n39), Alpha=Saturate(None=@n40))) {name=Eyes_Emissive, nodeColor=<rgba 0.1016 1 0 1>}
  Saturate(None=Add(A=Multiply(A=@n22) {constB=0.5}) {constB=0.5})
 ]}
