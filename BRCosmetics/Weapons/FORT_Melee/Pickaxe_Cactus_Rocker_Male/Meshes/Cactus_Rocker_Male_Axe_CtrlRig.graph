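# Control Rig for the Cactus Rocker Male pickaxe that adds procedural secondary motion to its dangling jaw bones.
# A single Chain Harmonics node runs from Begin Execution on the chain rooted at dyn_jaw_1, combining a pendulum with noisy wave motion on Z.
# The skeleton also has fish-like spine, fin and tail bones, but the graph does not drive them.
# Debug drawing is left enabled on the node.
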ControlRigBlueprint /BRCosmetics/Weapons/FORT_Melee/Pickaxe_Cactus_Rocker_Male/Meshes/Cactus_Rocker_Male_Axe_CtrlRig
graphs [
 RigVMFunctionLibrary {links [], nodes []}
 RigVMModel {nodes [
   RigUnit_BeginExecution = &RigUnit_BeginExecution()
   /Script/ControlRig.RigUnit_ChainHarmonicsPerItem(
    ChainRoot=(Type=Bone,Name="dyn_jaw_1")
    DrawWorldOffset=(Rotation=(X=0,Y=0,Z=0,W=1),Translation=(X=0,Y=0,Z=0),Scale3D=(X=1,Y=1,Z=1))
    Pendulum=(bEnabled=True,PendulumStiffness=2.5,PendulumGravity=(X=0,Y=0,Z=30),PendulumBlend=1,PendulumDrag=3,PendulumMinimum=0,PendulumMaximum=0.1,PendulumEase=Linear,UnwindAxis=(X=0,Y=1,Z=0),UnwindMinimum=0.2,UnwindMaximum=0.05)
    Reach=(bEnabled=True,ReachTarget=(X=0,Y=0,Z=0),ReachAxis=(X=0,Y=0,Z=0),ReachMinimum=0,ReachMaximum=0,ReachEase=Linear)
    Speed=(X=0,Y=0,Z=1)
    Wave=(bEnabled=True,WaveFrequency=(X=0,Y=0,Z=0),WaveAmplitude=(X=0,Y=0,Z=0.3),WaveOffset=(X=0,Y=0,Z=0),WaveNoise=(X=0,Y=0,Z=6),WaveMinimum=0,WaveMaximum=1,WaveEase=CubicEaseInOut)
    WaveCurve=(EditorCurveData=(Keys=((),(Time=1,Value=1))))
    bDrawDebug=True
    ExecutePin=@RigUnit_BeginExecution.ExecutePin
   ) {is=struct RigVMUnitNode}
  ]}
]
hierarchy [
 b_AR_Root {loc=<0 0 0>, rot=<0 0 0.7071 0.7071>}
 Melee_TwoHanded_Handle {parent=b_AR_Root, loc=<8.1477 -0.008 -29.0793>, rot=<0 0 0.7071 -0.7071>}
 dyn_spine_1 {parent=Melee_TwoHanded_Handle, loc=<1.1375 5.1942 74.2783>, rot=<0.0786 0.0786 0.7027 -0.7027>}
 dyn_R_fin {parent=dyn_spine_1, loc=<8.6396 -6.1015 -5.93>, rot=<0.1899 -0.6675 0.197 -0.6925>}
 dyn_L_fin {parent=dyn_spine_1, loc=<8.6396 6.1015 -5.93>, rot=<0.6925 0.197 -0.6675 -0.1899>}
 dyn_spine_2 {parent=dyn_spine_1, loc=<17.0731 0 0>, rot=<0 0 0 1>}
 dyn_spine_3 {parent=dyn_spine_2, loc=<11.7889 0 0>, rot=<0 0 0 1>}
 dyn_tail_1 {parent=dyn_spine_3, loc=<12.7732 0 0>, rot=<0 0 0 1>}
 dyn_tail_2 {parent=dyn_tail_1, loc=<11.1157 -0.0044 5.7048>, rot=<0 0 0 1>}
 dyn_tail_3 {parent=dyn_tail_2, loc=<10.9789 0 3.6493>, rot=<0 0 0 1>}
 dyn_jaw_1 {parent=dyn_spine_1, loc=<0.0032 0 -5.6817>, rot=<-0.1426 -0.6926 0.6926 -0.1426>}
 dyn_jaw_2 {parent=dyn_jaw_1, loc=<1.8089 0 0>, rot=<0 0 0 1>}
]
